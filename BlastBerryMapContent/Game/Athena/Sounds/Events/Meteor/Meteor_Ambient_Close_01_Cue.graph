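# Sound cue named as the close ambient for the meteor event in Athena, routed to an extra-high-priority event sound class with boosted volume.
# The file holds no root or sound nodes, so it carries only playback settings.
# Its attenuation override uses a custom curve that reaches silence a tenth of the way through a very long falloff, with wide stereo spread.
# Interpolated listener focus with narrow azimuths lowers volume when out of focus.
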SoundCue /BlastBerryMapContent/Game/Athena/Sounds/Events/Meteor/Meteor_Ambient_Close_01_Cue
root=null
nodes []
properties {
 bOverrideAttenuation=true
 maxDistance=101024
 soundClassObject=/BlastBerryMapContent/Game/Athena/Sounds/Events/SoundClasses/Fort_Events_Priority_ExtHigh
 volumeMultiplier=1.5
 attenuationOverrides {
  bEnableFocusInterpolation=true
  bEnableListenerFocus=true
  distanceAlgorithm=Custom
  falloffDistance=100000
  focusAttackInterpSpeed=0.25
  focusAzimuth=5
  focusDistanceScale=0.3
  focusReleaseInterpSpeed=0.25
  lPFFrequencyAtMax=500
  lPFRadiusMax=2000
  lPFRadiusMin=1000
  nonFocusAzimuth=10
  nonFocusVolumeAttenuation=0.5
  stereoSpread=2048
  attenuationShapeExtents {x=1024}
  customAttenuationCurve {editorCurveData {keys [<at 0 1>,<at 0.1 0>,<at 1 0>]}}
 }
}
