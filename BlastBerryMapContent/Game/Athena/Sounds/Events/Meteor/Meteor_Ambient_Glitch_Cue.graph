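# Sound cue for the meteor event's ambient glitch, mixing randomized one-shots with electrical ripple Doppler spots.
# A Mixer blends three layers, each random-picked, modulated and separately attenuated: dominant glitch one-shots, quieter ripples, and a quietest, randomly delayed, lower-pitched copy of the ripples.
# The ripple layers share the interior cave attenuation, while glitches use the ambient one.
# All quality levels share this Mixer, and a concurrency override caps simultaneous instances.
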SoundCue /BlastBerryMapContent/Game/Athena/Sounds/Events/Meteor/Meteor_Ambient_Glitch_Cue
root=QualityLevel_0
nodes [
 {
  id=Attenuation_0
  children [Modulator_1]
  properties {attenuationSettings=/Game/Sounds/Asteria/POI/Meteor/Spots/Meteor_Interior_Cave_Glitches_Att}
 }
 {id=Attenuation_1, children [Modulator_0], properties {attenuationSettings=/Game/Sounds/Asteria/POI/Meteor/Spots/Meteor_Ambient_Glitch_Att}}
 {
  id=Attenuation_2
  children [Modulator_2]
  properties {attenuationSettings=&Meteor_Interior_Cave_Glitches_Att}
 }
 {id=Delay_0, children [Attenuation_2], properties {delayMax=2, delayMin=0.5}}
 {id=Mixer_1, children [Attenuation_1, Attenuation_0, Delay_0], properties {inputVolume [1,0.2,0.1]}}
 {id=Modulator_0, children [Random_0], properties {pitchMax=1.1}}
 {id=Modulator_1, children [Random_2], properties {pitchMax=1.1, pitchMin=0.8}}
 {id=Modulator_2, children [Random_1], properties {pitchMax=0.9, pitchMin=0.8}}
 {id=QualityLevel_0, children [Mixer_1,Mixer_1,Mixer_1,Mixer_1]}
 {
  id=Random_0
  children [WavePlayer_0, WavePlayer_5, WavePlayer_6, WavePlayer_7, WavePlayer_8, WavePlayer_9, WavePlayer_10]
  properties {weights [1,1,1,1,1,1,1]}
 }
 {
  id=Random_1
  children [WavePlayer_15, WavePlayer_16, WavePlayer_17, WavePlayer_18, WavePlayer_19, WavePlayer_20]
  properties {weights [1,1,1,1,1,1]}
 }
 {
  id=Random_2
  children [WavePlayer_3, WavePlayer_4, WavePlayer_11, WavePlayer_12, WavePlayer_13, WavePlayer_14]
  properties {weights [1,1,1,1,1,1]}
 }
 {id=WavePlayer_0, properties {soundWaveAssetPtr=/BlastberryMapContent/Game/Athena/Sounds/Events/Meteor/Meteor_Glitch_Oneshot_01}}
 {id=WavePlayer_10, properties {soundWaveAssetPtr=/BlastberryMapContent/Game/Athena/Sounds/Events/Meteor/Meteor_Glitch_Oneshot_07}}
 {id=WavePlayer_11, properties {soundWaveAssetPtr=/Game/Sounds/Asteria/POI/Meteor/Spots/Spot_Meteor_Interior_Electrical_Ripple_Doppler_03}}
 {id=WavePlayer_12, properties {soundWaveAssetPtr=/Game/Sounds/Asteria/POI/Meteor/Spots/Spot_Meteor_Interior_Electrical_Ripple_Doppler_04}}
 {id=WavePlayer_13, properties {soundWaveAssetPtr=/Game/Sounds/Asteria/POI/Meteor/Spots/Spot_Meteor_Interior_Electrical_Ripple_Doppler_05}}
 {id=WavePlayer_14, properties {soundWaveAssetPtr=/Game/Sounds/Asteria/POI/Meteor/Spots/Spot_Meteor_Interior_Electrical_Ripple_Doppler_06}}
 {id=WavePlayer_15, properties {soundWaveAssetPtr=/Game/Sounds/Asteria/POI/Meteor/Spots/Spot_Meteor_Interior_Electrical_Ripple_Doppler_01}}
 {id=WavePlayer_16, properties {soundWaveAssetPtr=/Game/Sounds/Asteria/POI/Meteor/Spots/Spot_Meteor_Interior_Electrical_Ripple_Doppler_02}}
 {id=WavePlayer_17, properties {soundWaveAssetPtr=&Spot_Meteor_Interior_Electrical_Ripple_Doppler_03}}
 {id=WavePlayer_18, properties {soundWaveAssetPtr=&Spot_Meteor_Interior_Electrical_Ripple_Doppler_04}}
 {id=WavePlayer_19, properties {soundWaveAssetPtr=&Spot_Meteor_Interior_Electrical_Ripple_Doppler_05}}
 {id=WavePlayer_20, properties {soundWaveAssetPtr=&Spot_Meteor_Interior_Electrical_Ripple_Doppler_06}}
 {id=WavePlayer_3, properties {soundWaveAssetPtr=&Spot_Meteor_Interior_Electrical_Ripple_Doppler_01}}
 {id=WavePlayer_4, properties {soundWaveAssetPtr=&Spot_Meteor_Interior_Electrical_Ripple_Doppler_02}}
 {id=WavePlayer_5, properties {soundWaveAssetPtr=/BlastberryMapContent/Game/Athena/Sounds/Events/Meteor/Meteor_Glitch_Oneshot_02}}
 {id=WavePlayer_6, properties {soundWaveAssetPtr=/BlastberryMapContent/Game/Athena/Sounds/Events/Meteor/Meteor_Glitch_Oneshot_03}}
 {id=WavePlayer_7, properties {soundWaveAssetPtr=/BlastberryMapContent/Game/Athena/Sounds/Events/Meteor/Meteor_Glitch_Oneshot_04}}
 {id=WavePlayer_8, properties {soundWaveAssetPtr=/BlastberryMapContent/Game/Athena/Sounds/Events/Meteor/Meteor_Glitch_Oneshot_05}}
 {id=WavePlayer_9, properties {soundWaveAssetPtr=/BlastberryMapContent/Game/Athena/Sounds/Events/Meteor/Meteor_Glitch_Oneshot_06}}
]
properties {
 bOverrideConcurrency=true
 maxDistance=20000
 soundClassObject=&Fort_Events
 volumeMultiplier=1.25
 attenuationOverrides {
  bAttenuateWithLPF=true
  bEnableFocusInterpolation=true
  bEnableListenerFocus=true
  distanceAlgorithm=Logarithmic
  falloffDistance=20000
  focusAttackInterpSpeed=0.01
  focusDistanceScale=0.5
  focusReleaseInterpSpeed=0.01
  lPFFrequencyAtMax=500
  lPFRadiusMax=20000
  lPFRadiusMin=10000
  nonFocusVolumeAttenuation=0.5
  stereoSpread=4096
  attenuationShapeExtents {x=0}
  customAttenuationCurve {editorCurveData {keys [<at 0.1 0>,<at 0.25 1>,<at 1 0>]}}
 }
 concurrencyOverrides {maxCount=3}
}
